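# Translucent unlit Niagara material for scribble drift trails, drawing an animated flipbook or static texture with an optional scrolling background.
# The flipbook steps through eight rows over time, offset by a rounded dynamic parameter.
# Channel masks split the texture into color, outline and opacity, blending particle color with a panning rainbow LUT.
# Dynamic parameters scale outline and scribble opacity, while background opacity follows particle alpha with an optional ribbon-start fade.
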Material /VehicleCosmetics/DriftTrail/DriftTrail_Scribble/Materials/M_Ribbon_Flipbook
graph {
 properties {
  bAutomaticallySetUsageInEditor=false
  bUsedWithNiagaraMeshParticles=true
  bUsedWithNiagaraRibbons=true
  bUsedWithNiagaraSprites=true
  blendMode=BLEND_Translucent
  shadingModel=MSM_Unlit
  usedShadingModels=Unlit
 }
 nodes [
  n1 = TextureCoordinate() {uTiling=0.5}
  n2 = AppendVector(A=ScalarParameter() {defaultValue=0.5, parameterName=RainbowTile}, B=Constant())
  n3 = TextureSampleParameter2D(UVs=Panner(Coordinate=Multiply(A=@n1, B=@n2)) {speedX=-0.5}) {
   group=Emissive
   parameterName=RainbowLUT
   samplerSource=SSM_Wrap_WorldGroupSettings
   sortPriority=1
   texture=/Game/Global/Materials/Character_Parent_2023/FlipbookFace_Parent_2023/Textures/T_black_emptyalpha
  }
  n4 = TextureObjectParameter() {parameterName=MainTexture}
  n5 = TextureCoordinate()
  n6 = DynamicParameter() {defaultValue=<rgba 0 1 1 1>, paramNames [FlipbookOffset, OutlineStrength, ScribbleEndOpacity, Param4]}
  n7 = AppendVector(
   A=Constant()
   B=Clamp(None=MaterialFunctionCall!(In=@n6.FlipbookOffset) {materialFunction=/Engine/Functions/Engine_MaterialFunctions02/Utility/Round}.Result) {maxDefault=7}
  )
  n8 = StaticSwitchParameter(
   True=MaterialFunctionCall("Animation  Phase (0-1)"=Time(), Number of Rows=Constant() {r=8}, Number of Columns=Constant() {r=1}, Texture=@n4, UVs=Add(A=@n5, B=@n7)) {materialFunction=&FlipBook}.Result
   False=TextureSample(UVs=@n5, Tex=@n4).RGB
  ) {defaultValue=true, parameterName=FlipbookOrSprite}
  n9 = LinearInterpolate(B=@n3.RGB, Alpha=ChannelMaskParameter(None=@n8) {parameterName=ChannelMaskOutline}.R)
  n10 = DynamicParameter() {defaultValue=<rgba 0 1 1 1>, paramNames [FlipbookOffset, OutlineStrength, ScribbleEndOpacity, Param4]}
  n11 = ParticleColor()
  n12 = ChannelMaskParameter(None=@n8) {parameterName=ChannelMaskColor}
  n13 = ParticleColor()
  n14 = Multiply(
   A=TextureCoordinate()
   B=AppendVector(A=ScalarParameter() {defaultValue=1, parameterName=BackgroundScaleX}, B=ScalarParameter() {defaultValue=1, parameterName=BackgroundScaleY})
  )
  n15 = AppendVector(A=ScalarParameter() {defaultValue=-1, parameterName=BackgroundSpeed}, B=Constant())
  n16 = ChannelMaskParameter(None=TextureSampleParameter2D(UVs=Panner(Coordinate=@n14, Speed=@n15)) {parameterName=BackgroundTexture}.RGB) {parameterName=BackgroundChannel}
  n17 = StaticSwitchParameter(True=Add(A=@n12.R, B=Multiply(A=@n13.RGB, B=@n16.R)), False=@n12.R) {defaultValue=true, parameterName=UseBackground}
  MaterialFunctionCall(Emissive=Add(A=Multiply(A=@n9, B=@n10.OutlineStrength), B=LinearInterpolate(A=@n11.RGB, B=@n3.RGB, Alpha=@n17))) {materialFunction=&MF_FX_EmissiveAdjustments}
  n18 = NamedRerouteUsage()
  n19 = ComponentMask(None=TextureCoordinate() {coordinateIndex=1}) {r=true}
  NamedRerouteDeclaration(None=Saturate(None=Multiply(A=@n18, B=@n19))) {name=BG Ribbon Start Fade, nodeColor=<rgba 0.7363 0 1 1>}
  NamedRerouteDeclaration(None=@n16.R) {name=Background, nodeColor=<rgba 0 0.4375 1 1>}
  n20 = Multiply(
   A=ChannelMaskParameter(None=@n8) {parameterName=ChannelMaskOpacity}.R
   B=DynamicParameter() {defaultValue=<rgba 0 1 1 1>, paramNames [FlipbookOffset, OutlineStrength, ScribbleEndOpacity, Param4]}.ScribbleEndOpacity
  )
  n21 = Multiply(
   A=StaticSwitchParameter(True=NamedRerouteUsage(), False=@n16.R) {parameterName=UseStartFade}
   B=ScalarParameter() {defaultValue=0.9, parameterName=BackgroundAlpha}
  )
  StaticSwitchParameter(True=Add(A=@n20, B=Multiply(A=@n21, B=@n11.A)), False=@n20) {defaultValue=true, parameterName=UseBackground}
 ]
}
